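; Gameplay ability for the Rocket Drill that holsters the player's weapon through a gameplay effect until the pawn lands.
; Activation checks ignore tags and optional airborne-only movement, then applies the effect with a failsafe maximum duration.
; Landing, or tag waits for vehicle, delayed swimming, riding, sliding and grinding, serve as end triggers.
; A one-shot delayed end plays a montage masking the invisible-weapon pose if still holstered; ending removes the effect and waits.
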(event Custom|SetUpBinds
  ;[1 1 1] Set Up Binds
  (bind _asfort_pawn (Utilities|Casting|CastToFortPawn (Ability|GetAvatarActorfromActorInfo)))
  (EventDispatchers|BindEventtoOnPawnLanded _asfort_pawn (delegate Custom|PawnLanded)))

(event Custom|PawnLanded
  (params (Hit :type Hit%20Result%20Structure%20%28by%20ref%29))
  ;[1 1 1] Set Up Binds
  (CallFunction|CallEnd))

(event Custom|SetUpWaits
  ;[1 1 1] Set Up Waits
  (bind _asynctaskproxy (Ability|Tasks|WaitGameplayTagAdd (Variables|Default|GetTag_InVehicle))
    (:then
      (Utilities|Array|Add (Variables|Default|GetWaitTagAsyncTasks) _asynctaskproxy)
      (Utilities|FlowControl|Delay (Variables|Default|GetDelay_WaterSwimBind)) ; Don't do the water bind immidately, just incase the player doesn't leave the water in time.
      (bind _asynctaskproxy_2 (Ability|Tasks|WaitGameplayTagAdd (Variables|Default|GetTag_SurfaceSwimming))
        (:then
          (Utilities|Array|Add (Variables|Default|GetWaitTagAsyncTasks) _asynctaskproxy_2)
          (bind _asynctaskproxy_3 (Ability|Tasks|WaitGameplayTagAdd (Variables|Default|GetTagIsRiding))
            (:then
              (Utilities|Array|Add (Variables|Default|GetWaitTagAsyncTasks) _asynctaskproxy_3)
              (bind _asynctaskproxy_4 (Ability|Tasks|WaitGameplayTagAdd (Variables|Default|GetTag_Sliding))
                (:then
                  (Utilities|Array|Add (Variables|Default|GetWaitTagAsyncTasks) _asynctaskproxy_4)
                  (bind _asynctaskproxy_5 (Ability|Tasks|WaitGameplayTagAdd "(TagName=\"MovementMode.Grinding\")")
                    (:then
                      (Utilities|Array|Add (Variables|Default|GetWaitTagAsyncTasks) _asynctaskproxy_5))
                    (:Added
                      _)))
                (:Added
                  _)) ; When sliding down a hill or cliff that isn't walkable, the OnPawnLanded callback can get skipped.\nEnsure that if the player gets into a sliding state, they are treated as if they had landed.
              (CallFunction|CallEnd))
            (:Added
              _)) ; If the player start to ride creature we have to stop the Immunity too
        )
        (:Added
          _)))
    (:Added
      _)))

(event Ability|EventActivateAbility
  (bind _returnvalue (Ability|GetAvatarActorfromActorInfo))
  ;[1 1 1] Check if we have any ignore tags and if we pass then apply the falling damage immunity
  (Utilities|IsValid _returnvalue
    (:Is%20Valid
      (bind _returnvalue_2 (GameplayTags|HasAnyMatchingGameplayTags(Message) _returnvalue (Variables|Default|GetTC_IgnoreApplication)))
      (if (not _returnvalue_2)
        (if (Variables|Default|GetApplyInAirOnly)
          (bind _asfort_pawn (Utilities|Casting|CastToFortPawn (Ability|GetAvatarActorfromActorInfo))
            (:then
              (if (not (Math|Boolean|ANDBoolean (Utilities|Enum|NotEqual(Enum) (Variables|/Script/Engine.CharacterMovementComponent|GetMovementMode (Variables|/Script/Engine.Character|GetCharacterMovement _asfort_pawn)) "MOVE_Walking") (Utilities|Enum|NotEqual(Enum) (Variables|/Script/Engine.CharacterMovementComponent|GetMovementMode (Variables|/Script/Engine.Character|GetCharacterMovement _asfort_pawn)) "MOVE_NavWalking") (Utilities|Enum|NotEqual(Enum) (Variables|/Script/Engine.CharacterMovementComponent|GetMovementMode (Variables|/Script/Engine.Character|GetCharacterMovement _asfort_pawn)) "MOVE_Swimming")))
                (Ability|EndAbility)))
            (:CastFailed
              _))
          (else
            _))
        (bind _returnvalue_3 (Ability|GameplayEffect|AssignTagSetbyCallerMagnitude (Ability|MakeOutgoingGameplayEffectSpec (Variables|Default|GetHolsterGE)) "(TagName=\"SetByCaller.Duration\")" (Variables|Default|GetMaxHolsterTime%20%28Failsafe%29)))
        (bind _returnvalue_4 (Ability|ApplyGameplayEffectSpecToOwner _returnvalue_3))
        (Variables|Default|SetGE_Holster _returnvalue_4)
        ;;[1 1 1] Set up the waits and binds that will end the ability
        (CallFunction|SetUpWaits)
        (CallFunction|SetUpBinds)
        (else
          (Ability|EndAbility))))
    (:Is%20Not%20Valid
      (Ability|EndAbility))))

(event Custom|CallEnd
  ;[1 1 1] Call End - After a delay call end.
  (Utilities|FlowControl|DoOnce false
    (:Completed
      (Utilities|Time|SetTimerbyEvent (EventDispatchers|CreateEvent :!event K2_EndAbility) (ScalableFloat|GetValueatLevel (Variables|Default|GetDelayUntilEnd))) ; Needs to 
      (if (Weapon|IsWeaponHolstered (Ability|GetActivatingPawn))
        (Ability|Tasks|PlayMontageandWaitforTrigger "None" (Variables|Default|GetAnimToPlayOnLanding) (Variables|Default|GetAnimToPlayOnLanding_PlayRate)) ; Anim is needed to cover up the anim of the player holding their weapon when they haven't unholstered it,\nas if that anim plays while still holstered, it looks like the player is holding an invisible weapon
      ))))

(event Ability|EventOnEndAbility
  (params (bWasCancelled :type Boolean))
  (bind _waittagasynctasks (Variables|Default|GetWaitTagAsyncTasks))
  ;[1 1 1] On End Ability - Remove the Immunity
  (Ability|RemoveGameplayEffectFromOwnerWithHandle (Variables|Default|GetGE_Holster))
  (for _array_element _waittagasynctasks
    (Utilities|IsValid _array_element
      (:Is%20Valid
        (GameplayTasks|EndTask _array_element))))
  (Utilities|Array|Clear _waittagasynctasks))

(event Ability|EventFailedToActivatePassiveAbility
  (params (ActorInfo :type Gameplay%20Ability%20Actor%20Info%20Structure))
  (|CallParentFunction))
